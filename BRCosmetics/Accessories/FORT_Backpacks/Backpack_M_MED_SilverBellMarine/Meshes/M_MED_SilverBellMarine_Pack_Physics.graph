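# Physics asset for the SilverBellMarine backpack cosmetic, defining collision and joints for a small humanoid figure with an oversized head on the pack.
# Capsules cover the figure's dyn_ bones, and one box covers root.
# Head joins root with free rotation and is angularly locked to neck, so the body hangs from it.
# Limbs chain to pelvis through tight angular limits, with elbows and knees locking some axes.
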PhysicsAsset /BRCosmetics/Accessories/FORT_Backpacks/Backpack_M_MED_SilverBellMarine/Meshes/M_MED_SilverBellMarine_Pack_Physics
bodies {
 dyn_calf_l [{
   shapeType=Capsule
   length=1.5068
   radius=0.7644
   shapeName=dyn_calf_l_capsule
   center=<1.3475 0.0086 -0.0264>
   rotation=<rot 0 -89.3619 -87.015>
  }]
 dyn_calf_r [{
   shapeType=Capsule
   length=1.4124
   radius=0.7647
   shapeName=dyn_calf_r_capsule
   center=<-1.0477 0.5907 0.0001>
   rotation=<rot 0.3243 90.0377 -88.3802>
  }]
 dyn_foot_l [{
   shapeType=Capsule
   length=3.6347
   radius=1.7813
   shapeName=dyn_foot_l_capsule
   center=<0.0479 -1.6866 -0.6763>
   rotation=<rot 0.2115 -179.6708 -94.032>
  }]
 dyn_foot_r [{
   shapeType=Capsule
   length=3.6536
   radius=1.8476
   shapeName=dyn_foot_r_capsule
   center=<-0.0014 1.4324 0.7261>
   rotation=<rot 0.0184 178.5087 -90.9245>
  }]
 dyn_hand_l [{
   shapeType=Capsule
   length=0.4797
   radius=1.0893
   shapeName=dyn_hand_l_capsule
   center=<0.0347 -1.328 0.2462>
   rotation=<rot 3.1634 -67.1665 -101.9646>
  }]
 dyn_hand_r [{
   shapeType=Capsule
   length=0.6872
   radius=1.0122
   shapeName=dyn_hand_r_capsule
   center=<-0.0503 1.3232 -0.2455>
   rotation=<rot 3.1269 -59.6434 -93.4685>
  }]
 dyn_head [{shapeType=Capsule, length=1.0781, radius=11.5008, center=<-3.0155 0.24 -0.0348>, rotation=<rot 0.1562 -90 -0.655>}]
 dyn_lowerarm_l [{
   shapeType=Capsule
   length=1.3135
   radius=0.6927
   shapeName=dyn_lowerarm_l_capsule
   center=<1.477 -0.5872 0.0086>
   rotation=<rot 0 87.3284 -89.0078>
  }]
 dyn_neck [{shapeType=Capsule, length=2.8135, radius=1.0461, shapeName=dyn_neck_capsule, rotation=<rot 0 -90 0>}]
 dyn_pelvis [{shapeType=Capsule, length=2.7861, radius=2.8278, center=<0.0376 0.2288 -1.4147>, rotation=<rot 0.0328 -89.8297 -1.2711>}]
 dyn_rowerarm_r [{
   shapeType=Capsule
   length=1.3114
   radius=0.6876
   shapeName=dyn_rowerarm_r_capsule
   center=<-1.44 0.0482 -0.0128>
   rotation=<rot 0 88.9817 -87.9864>
  }]
 dyn_thigh_l [{
   shapeType=Capsule
   length=1.006
   radius=1.3955
   shapeName=dyn_thigh_l_capsule
   center=<1.9814 -0.3775 0.1581>
   rotation=<rot 19.4449 -90 88.8113>
  }]
 dyn_thigh_r [{
   shapeType=Capsule
   length=1.2264
   radius=1.3611
   shapeName=dyn_thigh_r_capsule
   center=<-2.0389 0.0692 -0.1083>
   rotation=<rot 9.7734 -86.1254 -91.4329>
  }]
 dyn_upperarm_l [{
   shapeType=Capsule
   length=1.81
   radius=1.344
   shapeName=dyn_upperarm_l_capsule
   center=<1.294 -0.1498 0.048>
   rotation=<rot 0 95.1405 -85.8956>
  }]
 dyn_upperarm_r [{
   shapeType=Capsule
   length=1.8449
   radius=1.3062
   shapeName=dyn_upperarm_r_capsule
   center=<-1.376 0.5204 -0.0187>
   rotation=<rot 0 98.8192 -84.5684>
  }]
 root [{shapeType=Box, extentX=3.23, extentY=15.0137, extentZ=15.1427, shapeName=root_box, center=<0 -0.3949 -12.8161>, rotation=<rot 0 -90 0>}]
}
constraints [
 {bone1Name=dyn_pelvis, bone2Name=dyn_neck, swing1LimitDegrees=45, swing1Motion=Locked, swing2LimitDegrees=1, twistLimitDegrees=1}
 {bone1Name=dyn_hand_r, bone2Name=dyn_rowerarm_r, swing1LimitDegrees=3, swing2LimitDegrees=1, twistLimitDegrees=3}
 {bone1Name=dyn_rowerarm_r, bone2Name=dyn_upperarm_r, swing1LimitDegrees=3, swing2LimitDegrees=13, twistLimitDegrees=45, twistMotion=Locked}
 {bone1Name=dyn_upperarm_r, bone2Name=dyn_pelvis, swing1LimitDegrees=33, swing2LimitDegrees=23, twistLimitDegrees=2}
 {bone1Name=dyn_hand_l, bone2Name=dyn_lowerarm_l, swing1LimitDegrees=2, swing2LimitDegrees=1, twistLimitDegrees=2}
 {bone1Name=dyn_lowerarm_l, bone2Name=dyn_upperarm_l, swing1LimitDegrees=5, swing2LimitDegrees=18, twistLimitDegrees=45, twistMotion=Locked}
 {bone1Name=dyn_upperarm_l, bone2Name=dyn_pelvis, swing1LimitDegrees=28, swing2LimitDegrees=33, twistLimitDegrees=1}
 {bone1Name=dyn_foot_r, bone2Name=dyn_calf_r, swing1LimitDegrees=6, swing2LimitDegrees=2, twistLimitDegrees=5}
 {
  bone1Name=dyn_calf_r
  bone2Name=dyn_thigh_r
  swing1LimitDegrees=2
  swing1Motion=Locked
  swing2LimitDegrees=20
  twistLimitDegrees=2
  twistMotion=Locked
 }
 {bone1Name=dyn_thigh_r, bone2Name=dyn_pelvis, swing1LimitDegrees=17, swing2LimitDegrees=8, twistLimitDegrees=1}
 {bone1Name=dyn_foot_l, bone2Name=dyn_calf_l, swing1LimitDegrees=4, swing2LimitDegrees=2, twistLimitDegrees=3}
 {
  bone1Name=dyn_calf_l
  bone2Name=dyn_thigh_l
  swing1LimitDegrees=2
  swing1Motion=Locked
  swing2LimitDegrees=20
  twistLimitDegrees=2
  twistMotion=Locked
 }
 {bone1Name=dyn_thigh_l, bone2Name=dyn_pelvis, swing1LimitDegrees=11, swing2LimitDegrees=15, twistLimitDegrees=2}
 {
  bone1Name=dyn_neck
  bone2Name=dyn_head
  swing1LimitDegrees=45
  swing1Motion=Locked
  swing2LimitDegrees=45
  swing2Motion=Locked
  twistLimitDegrees=45
  twistMotion=Locked
 }
 {
  bone1Name=dyn_head
  bone2Name=root
  swing1LimitDegrees=45
  swing1Motion=Free
  swing2LimitDegrees=45
  swing2Motion=Free
  twistLimitDegrees=45
  twistMotion=Free
 }
]
