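# Animation Blueprint for the SilverBellMarine backpack cosmetic, adding physics motion and repositioning the pack during skydiving and parachuting.
# A RigidBody node simulates in root bone space, then chained additive Transform Bone nodes offset the pack bone.
# Each update, skydive or parachute state lowers the physics alpha and enables the skydive offset.
# That state also gates two interpolated offsets driven by spine bend, measured as yaw delta from the reference pose.
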AnimBlueprint /BRCosmetics/Accessories/FORT_Backpacks/Backpack_M_MED_SilverBellMarine/Meshes/M_MED_SilverBellMarine_Pack_AnimBP
target_skeleton=&/M_MED_SilverBellMarine_Pack_Skeleton
graphs [
 AnimGraph {nodes [
   n1 = RigidBody(ComponentPose=Local To Component(LocalPose=Input Pose()), Alpha=Get Rigid_Body_Alpha() -> (Rigid_Body_Alpha:double)) {
    bOverrideWorldGravity=true
    overridePhysicsAsset=&/M_MED_SilverBellMarine_Pack_Physics
    simulationSpace=BaseBoneSpace
    alphaBoolBlend {blendOption=Linear}
    baseBoneRef {boneName=root}
    componentAppliedLinearAccClamp=<5000 5000 5000>
    componentLinearAccScale=<0.7 0.7 0.7>
    componentLinearVelScale=<0.7 0.7 0.7>
    overrideWorldGravity {z=-980}
    simSpaceSettings {maxAngularAcceleration=450, maxAngularVelocity=35, maxLinearAcceleration=150, maxLinearVelocity=10, velocityScaleZ=0, worldAlpha=0.85}
   }
   n2 = "Transform (Modify) Bone - Bone: pack"!(ComponentPose=@n1, bAlphaBoolEnabled:bool, Translation="0,1.7,0", Rotation="0,0,0", Scale) -> (Pose) {
    alphaInputType=Bool
    translationMode=BMM_Additive
    translationSpace=BCS_BoneSpace
    alphaBoolBlend {blendInTime=1, blendOption=Sinusoidal, blendOutTime=1}
    boneToModify {boneName=pack}
    translation {y=1.7}
   }
   n3 = "Transform (Modify) Bone - Bone: pack"(ComponentPose=@n2, bAlphaBoolEnabled=Get Skydive_Enabled() -> (Skydive_Enabled:bool), Translation="0,0,-8", Rotation="0,0,0") {
    alphaInputType=Bool
    translationMode=BMM_Additive
    translationSpace=BCS_BoneSpace
    alphaBoolBlend {blendInTime=1, blendOption=Sinusoidal, blendOutTime=1}
    boneToModify {boneName=pack}
    translation {z=-8}
   }
   n4 = "Transform (Modify) Bone - Bone: pack"!(
    ComponentPose=@n3
    Alpha:float=Get Backpack_Move_Negative_Alpha() -> (Backpack_Move_Negative_Alpha:double)
    Translation="0,-18,50"
    Rotation="0,0,0"
    Scale
   ) -> (Pose) {
    translationMode=BMM_Additive
    translationSpace=BCS_BoneSpace
    alphaBoolBlend {blendInTime=2, blendOption=Sinusoidal, blendOutTime=2}
    alphaScaleBiasClamp {bInterpResult=true, interpSpeedDecreasing=1, interpSpeedIncreasing=1}
    boneToModify {boneName=pack}
    translation {y=-18, z=50}
   }
   n5 = "Transform (Modify) Bone - Bone: pack"(
    ComponentPose=@n4
    Alpha=Get Backpack_Move_Positive_Alpha() -> (Backpack_Move_Positive_Alpha:double)
    Translation="0,-10,20"
    Rotation="0,0,0"
   ) {
    translationMode=BMM_Additive
    translationSpace=BCS_BoneSpace
    alphaBoolBlend {blendInTime=2, blendOption=Sinusoidal, blendOutTime=2}
    alphaScaleBiasClamp {bInterpResult=true, interpSpeedDecreasing=1, interpSpeedIncreasing=1}
    boneToModify {boneName=pack}
    translation {y=-10, z=20}
   }
   Output Pose(Result=Component To Local(ComponentPose=@n5)) {layerGroup=None}
  ]}
 EventGraph {nodes [
   Event BlueprintUpdateAnimation()
   > n1 = Sequence()
   n2 = BreakTransform(InTransform=GetDeltaTransformFromRefPose(self=GetOwnerSkeletalMeshComponent(), BoneName=spine_05, BaseName=spine_01))
   n3 = Set Backpack_Move_Negative(execute=@n1.then_1, Backpack_Move_Negative:double=MapRangeClamped(Value=BreakRotator(InRot=@n2.Rotation).Yaw, InRangeA=-5, InRangeB=-55)) -> (Output_Get:double)
   n4 = Set Skydive_Enabled(
    execute=@n1.then_0
    Skydive_Enabled:bool=OR Boolean!(A=Get bIsSkydiveFloating(), B=Get bIsSkydiveDiveMode(), C:bool=Get bIsParachuteOpened()) -> (ReturnValue)
   ) -> (Output_Get:bool)
   n5 = "To Float (Boolean)"(InBool=@n4)
   Set Backpack_Move_Negative_Alpha(execute=@n3.then, Backpack_Move_Negative_Alpha:double=float * float(A=@n3, B=@n5) {is=title K2Node_PromotableOperator}) -> (Output_Get:double)
   n6 = BreakTransform(InTransform=GetDeltaTransformFromRefPose(self=GetOwnerSkeletalMeshComponent(), BoneName=spine_03, BaseName=spine_01))
   > n7 = Set Backpack_Move_Positive(Backpack_Move_Positive:double=MapRangeClamped(Value=BreakRotator(InRot=@n6.Rotation).Yaw, InRangeA=0, InRangeB=20)) -> (Output_Get:double)
   > Set Backpack_Move_Positive_Alpha(Backpack_Move_Positive_Alpha:double=float * float(A=@n7, B=@n5)) -> (Output_Get:double)
   TryGetPawnOwner()
   Set Rigid_Body_Alpha(execute=@n8.else, Rigid_Body_Alpha:double=0.75) -> (Output_Get:double)
   Set Rigid_Body_Alpha(execute=@n8.then, Rigid_Body_Alpha=0.15)
   n8 = Branch(execute=@n4.then, Condition=@n4)
  ]}
]
variables [
 Backpack_Move_Negative {type=number}
 Backpack_Move_Negative_Alpha {type=number}
 Backpack_Move_Positive {type=number}
 Backpack_Move_Positive_Alpha {type=number}
 Rigid_Body_Alpha {type=number, default=0.75}
 Skydive_Enabled {type=boolean}
]
